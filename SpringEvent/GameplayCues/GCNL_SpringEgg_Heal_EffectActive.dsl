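; Looping gameplay cue for the SpringEgg heal effect that sets its VFX to show either health healing or shielding.
; PassHealingParametersToVFX maps the Healing or Shield source tag to the User.isShielding bool on the first looping particle component.
; To handle remote clients sometimes running LoopingStart before Application, OnApplication caches the tags or configures already stored particles.
; LoopingStart merges cached and incoming tags, applies them, and stores its particles for late setup.
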(event EventOnLoopingStart(EitherVFX-preferthisversion)
  (params (MyTarget :type Actor%20Object%20Reference) (Parameters :type Gameplay%20Cue%20Parameters%20Structure%20%28by%20ref%29) (ParticleComponents :type Array%20of%20FXSystem%20Component%20Object%20References) (AudioComponents :type Array%20of%20Audio%20Component%20Object%20References))
  (bind (_normalizedmagnitude _rawmagnitude _effectcontext _matchedtagname _originaltag _aggregatedsourcetags) (Ability|GameplayCue|BreakGameplayCueParameters Parameters))
  (bind _output (Utilities|Array|Get(acopy) ParticleComponents))
  (bind _cachedaggregatedsourcetags (Variables|Default|GetCachedAggregatedSourceTags))
  ;[0.5061 0 1] AggregatedSourceTags will pass in a tag to determine of the active effect is healing HEALTH or SHIELD
  (|CallParentFunction MyTarget Parameters ParticleComponents AudioComponents)
  (GameplayTags|AppendGameplayTagContainers _cachedaggregatedsourcetags _aggregatedsourcetags) ; combine possible tags in case \nloop runs first and has valid tags
  (Utilities|IsValid _output
    (:Is%20Valid
      (if (Utilities|Array|IsNotEmpty (GameplayTags|BreakGameplayTagContainer _cachedaggregatedsourcetags))
        (bind _wassucessful (CallFunction|PassHealingParameterstoVFX _output _cachedaggregatedsourcetags))
        (if _wassucessful
          _
          (else
            _))
        (else
          _))
      ;;[0.9143 0 1] On Remote Clients, sometimes Loop is run before Application and the parameters haven't been set up properly. Store the Looping particles so they can have their parameters set as expected
      (Variables|Default|SetLoopingParticleComponents ParticleComponents))))

(event EventOnApplication(EitherVFX-preferthisone)
  (params (MyTarget :type Actor%20Object%20Reference) (Parameters :type Gameplay%20Cue%20Parameters%20Structure%20%28by%20ref%29) (ParticleComponents :type Array%20of%20FXSystem%20Component%20Object%20References) (AudioComponents :type Array%20of%20Audio%20Component%20Object%20References) (BurstCameraShakeInstance :type Camera%20Shake%20Base%20Object%20Reference) (BurstDecalInstance :type Decal%20Actor%20Object%20Reference))
  (bind (_normalizedmagnitude _rawmagnitude _effectcontext _matchedtagname _originaltag _aggregatedsourcetags) (Ability|GameplayCue|BreakGameplayCueParameters Parameters))
  (bind _loopingparticlecomponents (Variables|Default|GetLoopingParticleComponents))
  (bind _output (Utilities|Array|Get(acopy) _loopingparticlecomponents))
  ;[1 0 0.9619] OnApplication
  ; As noted above, this GCNL should not actually use Application Effects, but is necessary to ensure that the properties are valid when trying to set the VFX properties in LoopingStart
  (|CallParentFunction MyTarget Parameters ParticleComponents AudioComponents BurstCameraShakeInstance BurstDecalInstance)
  (if (Utilities|Array|IsEmpty _loopingparticlecomponents)
    (Variables|Default|SetCachedAggregatedSourceTags _aggregatedsourcetags) ; cache the tag parameters for loopingStart
    (else
      (Utilities|IsValid _output
        (:Is%20Valid
          (CallFunction|PassHealingParameterstoVFX _output _aggregatedsourcetags) ; set up the healing VFX because LoopingStart \ndidn't have the tags yet
        )))))

(fn PassHealingParametersToVFX
  (params (VFX :type FXSystem%20Component%20Object%20Reference) (HealingTag :type Gameplay%20Tag%20Container%20Structure%20%28by%20ref%29))
  (if (GameplayTags|HasTag HealingTag "(TagName=\"Gameplay.SpringEgg.Heal.AbilityEffect.Healing\")")
    (Effects|Components|ParticleSystem|SetBoolParameter VFX "User.isShielding" false)
    (return true)
    (elif (GameplayTags|HasTag HealingTag "(TagName=\"Gameplay.SpringEgg.Heal.AbilityEffect.Shield\")")
      (Effects|Components|ParticleSystem|SetBoolParameter VFX "User.isShielding" true)
      (return true)
      (else
        (return false)))))

(fn ConstructionScript)
