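; Burst gameplay cue notify for egg laying in the spring event, which tints the spawned VFX to match the type of egg.
; OnBurst treats the cue Target's class as the egg class and looks it up in EggToColorIndexMap.
; The resulting index sets the "ChickenColor" int parameter on the first spawned FX component, if valid.
; A Sequence guarantees the return runs regardless of the validity checks.
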(fn OnBurst
  (params (Target :type Actor%20Object%20Reference) (Parameters :type Gameplay%20Cue%20Parameters%20Structure%20%28by%20ref%29) (SpawnResults :type Gameplay%20Cue%20Notify%20Spawn%20Result%20Structure%20%28by%20ref%29))
  (bind _fxsystemcomponents (Utilities|Struct|BreakGameplayCueNotifySpawnResult SpawnResults))
  (bind _output (Utilities|Array|Get(acopy) _fxsystemcomponents))
  ;[0.4143 0 1] OnBurst
  ; Determine the type of egg spawned to customize the VFX color to match
  (Utilities|FlowControl|Sequence
    (:then_0
      (Utilities|IsValid Target
        (:Is%20Valid
          (bind _output_get (LocalVariables|SetEggClass (Utilities|GetClass Target)))
          (bind _output_get_2 (LocalVariables|SetEggIndex (Utilities|Map|Find (Variables|Default|GetEggToColorIndexMap) _output_get)))
          (Utilities|IsValid _output
            (:Is%20Valid
              (Effects|Components|ParticleSystem|SetIntParameter _output "ChickenColor" _output_get_2))))))
    (:then_1
      ;;[0.9466 0.456 1] Return: Use sequence to guarantee call
      (return))) ; use sequence to guarantee Return call
)
