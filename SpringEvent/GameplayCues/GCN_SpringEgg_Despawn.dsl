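; Gameplay cue notify that handles the burst effect when a SpringEvent egg despawns, varying the particle effect by egg type.
; OnBurst compares the target actor's class with the Heal and Hop egg classes to pick an integer egg index.
; It passes that index as an int parameter to the first particle component in the spawn results, though the parameter name is set to None.
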(fn OnBurst
  (params (Target :type Actor%20Object%20Reference) (Parameters :type Gameplay%20Cue%20Parameters%20Structure%20%28by%20ref%29) (SpawnResults :type Gameplay%20Cue%20Notify%20Spawn%20Result%20Structure%20%28by%20ref%29))
  (bind _output_get (LocalVariables|SetEggClass (Utilities|GetClass Target)))
  (if (== _output_get "/SpringEvent/Items/HealEgg/CBGA_Gathered_SpringEgg_Heal.CBGA_Gathered_SpringEgg_Heal_C")
    (LocalVariables|SetEggIndex 0)
    (elif (== _output_get "/SpringEvent/Items/HopEgg/CBGA_SpringEgg_Hops.CBGA_SpringEgg_Hops_C")
      (LocalVariables|SetEggIndex 1)))
  (Effects|Components|ParticleSystem|SetIntParameter (Utilities|Array|Get(acopy) (Utilities|Struct|BreakGameplayCueNotifySpawnResult SpawnResults)) "None" (LocalVariables|GetEggIndex))
  (return))
