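; Gameplay cue burst notify for the end of the SpringEgg heal effect, tailoring its particle effect to whether health or shield was restored.
; On burst, it takes the first spawned particle component, if valid, and checks the cue's aggregated source tags.
; The Healing tag sets the User.isShielding particle parameter false and the Shield tag sets it true; with neither tag, the parameter stays untouched.
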(fn OnBurst(EitherVFX-preferthisversion)
  (params (MyTarget :type Actor%20Object%20Reference) (Parameters :type Gameplay%20Cue%20Parameters%20Structure%20%28by%20ref%29) (ParticleComponents :type Array%20of%20FXSystem%20Component%20Object%20References) (AudioComponents :type Array%20of%20Audio%20Component%20Object%20References) (BurstCameraShakeInstance :type Camera%20Shake%20Base%20Object%20Reference) (BurstDecalInstance :type Decal%20Actor%20Object%20Reference))
  (bind (_normalizedmagnitude _rawmagnitude _effectcontext _matchedtagname _originaltag _aggregatedsourcetags) (Ability|GameplayCue|BreakGameplayCueParameters Parameters))
  (bind _output (Utilities|Array|Get(acopy) ParticleComponents))
  ;[0.5061 0 1] AggregatedSourceTags will pass in a tag to determine of the active effect is healing HEALTH or SHIELD
  (Utilities|IsValid _output
    (:Is%20Valid
      (if (GameplayTags|HasTag _aggregatedsourcetags "(TagName=\"Gameplay.SpringEgg.Heal.AbilityEffect.Healing\")")
        (Effects|Components|ParticleSystem|SetBoolParameter _output "User.isShielding" false)
        (elif (GameplayTags|HasTag _aggregatedsourcetags "(TagName=\"Gameplay.SpringEgg.Heal.AbilityEffect.Shield\")")
          (Effects|Components|ParticleSystem|SetBoolParameter _output "User.isShielding" true)
          (else
            _)))
      (return))))
